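# Physics asset for the Medusa pickaxe mesh, a melee weapon cosmetic.
# It defines a single body on the Melee_TwoHanded_Handle bone, shaped as one wide capsule that is offset and rotated relative to that bone.
# It has no constraints, so the asset describes one simple collision volume with no jointed parts.
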PhysicsAsset /BRCosmetics/Weapons/FORT_Melee/Pickaxe_Medusa/Meshes/Pickaxe_Medusa_PhysicsAsset
bodies {Melee_TwoHanded_Handle [{shapeType=Capsule, length=67.5678, radius=39.5923, center=<0.593 -1.016 25.0111>, rotation=<rot 0.0206 -90 -0.1367>}]}
constraints []
